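# SoundCue for the Void activation sound effect of VividRazor minions, playing a single wave named for the Baron's Void activation.
# The wave player feeds a Modulator left at default settings, with the cue's overall volume reduced.
# It uses enemy NPC void-burst attenuation and the gadgets/items sound class, and overrides concurrency to cap simultaneous instances.
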SoundCue /VividRazor_Minions/Audio/SFX/SC_VividRazor_Minions_Activate_Void
root=Modulator_0
nodes [
 {id=Modulator_0, children [WavePlayer_0]}
 {id=WavePlayer_0, properties {soundWaveAssetPtr=&/VividRazor_Baron_Activate_Void}}
]
properties {
 attenuationSettings=&/ATT_VoidBurstOpen_NPC_Enemy
 bOverrideConcurrency=true
 maxDistance=6000
 soundClassObject=&Fort_Gadgets_Items
 volumeMultiplier=0.7
 concurrencyOverrides {maxCount=4}
}
