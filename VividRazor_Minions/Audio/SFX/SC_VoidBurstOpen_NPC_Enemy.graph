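# Sound cue for the enemy NPC void burst open effect in VividRazor Minions.
# A Random node picks one of three equally weighted wave variations and feeds it through a Modulator at the root.
# It uses its own attenuation asset, plays in the gadgets and items sound class, and overrides concurrency to cap simultaneous instances.
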SoundCue /VividRazor_Minions/Audio/SFX/SC_VoidBurstOpen_NPC_Enemy
root=Modulator_0
nodes [
 {id=Modulator_0, children [Random_0]}
 {id=Random_0, children [WavePlayer_1,WavePlayer_2,WavePlayer_3], properties {weights [1,1,1]}}
 {id=WavePlayer_1, properties {soundWaveAssetPtr=&/VoidBurstOpen_NPC_Enemy_01}}
 {id=WavePlayer_2, properties {soundWaveAssetPtr=&/VoidBurstOpen_NPC_Enemy_02}}
 {id=WavePlayer_3, properties {soundWaveAssetPtr=&/VoidBurstOpen_NPC_Enemy_03}}
]
properties {
 attenuationSettings=&/ATT_VoidBurstOpen_NPC_Enemy
 bOverrideConcurrency=true
 maxDistance=6000
 soundClassObject=&Fort_Gadgets_Items
 volumeMultiplier=1
 concurrencyOverrides {maxCount=4}
}
