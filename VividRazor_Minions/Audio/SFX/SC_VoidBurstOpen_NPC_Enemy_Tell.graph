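# Sound cue for the Void Burst Open enemy NPC tell sound among the VividRazor minion SFX.
# A mixer layers two modulated random branches: one picks from four main tell waves, the other from six quieter ghost variants at half volume.
# Custom attenuation, a gadgets and items sound class, and overridden concurrency with a low instance cap and short retrigger time govern playback.
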SoundCue /VividRazor_Minions/Audio/SFX/SC_VoidBurstOpen_NPC_Enemy_Tell
root=Mixer_0
nodes [
 {id=Mixer_0, children [Modulator_0,Modulator_1], properties {inputVolume [1,0.5]}}
 {id=Modulator_0, children [Random_1]}
 {id=Modulator_1, children [Random_2]}
 {
  id=Random_1
  children [WavePlayer_0,WavePlayer_7,WavePlayer_8,WavePlayer_9]
  properties {bShouldExcludeFromBranchCulling=true, weights [1,1,1,1]}
 }
 {
  id=Random_2
  children [WavePlayer_10, WavePlayer_11, WavePlayer_12, WavePlayer_13, WavePlayer_14, WavePlayer_15]
  properties {bShouldExcludeFromBranchCulling=true, weights [1,1,1,1,1,1]}
 }
 {id=WavePlayer_0, properties {soundWaveAssetPtr=&/VoidBurstOpen_NPC_Enemy_Tell_01}}
 {id=WavePlayer_10, properties {soundWaveAssetPtr=&/VoidBurstOpen_NPC_Enemy_Tell_Ghost_01}}
 {id=WavePlayer_11, properties {soundWaveAssetPtr=&/VoidBurstOpen_NPC_Enemy_Tell_Ghost_02}}
 {id=WavePlayer_12, properties {soundWaveAssetPtr=&/VoidBurstOpen_NPC_Enemy_Tell_Ghost_03}}
 {id=WavePlayer_13, properties {soundWaveAssetPtr=&/VoidBurstOpen_NPC_Enemy_Tell_Ghost_04}}
 {id=WavePlayer_14, properties {soundWaveAssetPtr=&/VoidBurstOpen_NPC_Enemy_Tell_Ghost_05}}
 {id=WavePlayer_15, properties {soundWaveAssetPtr=&/VoidBurstOpen_NPC_Enemy_Tell_Ghost_06}}
 {id=WavePlayer_7, properties {soundWaveAssetPtr=&/VoidBurstOpen_NPC_Enemy_Tell_02}}
 {id=WavePlayer_8, properties {soundWaveAssetPtr=&/VoidBurstOpen_NPC_Enemy_Tell_03}}
 {id=WavePlayer_9, properties {soundWaveAssetPtr=&/VoidBurstOpen_NPC_Enemy_Tell_04}}
]
properties {
 attenuationSettings=&/ATT_VoidBurstOpen_NPC_Enemy
 bOverrideConcurrency=true
 maxDistance=6000
 soundClassObject=&Fort_Gadgets_Items
 volumeMultiplier=0.7
 concurrencyOverrides {maxCount=2, retriggerTime=0.1}
}
